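# Master translucent, unlit, two-sided material for the Rift Beacon bubble, with near and far (HLOD) looks.
# Near look layers panning turbulence with chromatic UV offsets and noise-driven BumpOffset, adding depth-fade fill and shaping opacity by edge, Fresnel, distance, time of day and facing.
# Far look adds Fresnel glow to cheaper turbulence; each look collapses via WPO on opposite sides of a distance threshold.
# Both support optional TacoTime color effects.
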Material /Figment_S10/Athena/Sounds/Events/RiftBeacon/RiftBubble/M_RiftBubbleMaster
graph {
 properties {
  bUseMaterialAttributes=true
  blendMode=BLEND_Translucent
  pixelDepthOffsetMode=PDOM_Legacy
  shadingModel=MSM_Unlit
  twoSided=true
  usedShadingModels=Unlit
 }
 nodes [
  n1 = TextureCoordinate() {uTiling=3}
  n2 = Multiply(A=Time(), B=ScalarParameter() {defaultValue=0.25, parameterName=TexturePanningSpeed2})
  n3 = Subtract(
   A=TextureSample(UVs=Panner(Coordinate=@n1, Time=@n2) {speedX=0.2, speedY=0.2}) {samplerType=SAMPLERTYPE_LinearColor, texture=&T_NoisePack_LinearColor}.R
   B=TextureSample(UVs=Panner(Coordinate=@n1, Time=@n2) {speedX=-0.2, speedY=-0.2}) {samplerType=SAMPLERTYPE_LinearColor, texture=&T_NoisePack_LinearColor}.G
  )
  n4 = DepthFade(FadeDistance=LinearInterpolate(
    A=ScalarParameter() {defaultValue=100, parameterName=DepthFadeInverse}
    B=ScalarParameter() {defaultValue=500, parameterName=DepthFadeInverse2}
    Alpha=Clamp(None=@n3)
   ))
  n5 = Constant() {r=1}
  n6 = Add(
   A=OneMinus(None=ShadingPathSwitch(Default=@n4, Mobile=@n5) {desc=Migrated from FeatureLevelSwitch})
   B=ScalarParameter() {defaultValue=0.6, parameterName=OpacityAdd}
  ) {constB=0.1}
  n7 = ScalarParameter() {defaultValue=3, parameterName=DepthPower}
  n8 = AppendVector(A=ScalarParameter() {parameterName=Chroma_UOffset}, B=ScalarParameter() {defaultValue=-0.05, parameterName=Chroma_VOffset2})
  n9 = TextureCoordinate()
  n10 = AppendVector(A=ScalarParameter() {defaultValue=6, parameterName=TextureA_UScale}, B=ScalarParameter() {defaultValue=2, parameterName=TextureA_VScale})
  n11 = TextureCoordinate() {uTiling=3}
  n12 = Multiply(A=Time(), B=ScalarParameter() {defaultValue=0.025, parameterName=TexturePanningSpeed})
  n13 = TextureSample(UVs=Panner(Coordinate=@n11, Time=@n12) {speedX=0.2, speedY=0.2}) {samplerType=SAMPLERTYPE_LinearColor, texture=&T_NoisePack_LinearColor}
  n14 = TextureSample(UVs=Panner(Coordinate=@n11, Time=@n12) {speedX=-0.2, speedY=-0.2}) {samplerType=SAMPLERTYPE_LinearColor, texture=&T_NoisePack_LinearColor}
  n15 = ComponentMask(None=TextureCoordinate() {uTiling=0.5, vTiling=0.5}) {g=true}
  n16 = MaterialFunctionCall(In=Abs(None=Sine(None=@n15)), Contrast=Constant() {r=0.1}) {materialFunction=&CheapContrast}
  n17 = VertexNormalWS()
  n18 = CameraVectorWS()
  n19 = CurveAtlasRowParameter(CurveTime=Abs(None=DotProduct(A=@n17, B=@n18))) {
   atlas=&AthenaCurveAtlas01
   curve=/Game/Athena/Environments/Nexus/Materials/FXshader/Curve_Ramps
   defaultValue=79
  }
  n20 = ScalarParameter() {defaultValue=0.025, parameterName=TexturePanningSpeed}
  n21 = Add(
   A=Multiply(A=Subtract(A=@n13.R, B=@n14.G), B=@n16.Result)
   B=Multiply(
    A=Subtract(A=@n19.R)
    B=MaterialFunctionCall(Speed=@n20) {materialFunction=&DebugTimeSine}.Result
   )
  )
  n22 = Multiply(A=@n16.Result, B=ScalarParameter() {defaultValue=0.25, parameterName=Height})
  n23 = TextureCoordinate() {uTiling=9, vTiling=3}
  n24 = Multiply(A=Time(), B=ScalarParameter() {defaultValue=0.025, parameterName=TexturePanningSpeed})
  n25 = Multiply(
   A=TextureSample(UVs=Panner(Coordinate=@n23, Time=@n24) {speedX=-0.02, speedY=-0.03}) {samplerType=SAMPLERTYPE_Grayscale, texture=&LowResBlurredNoise}.R
   B=Constant2Vector() {g=0.14, r=0.07}
  )
  n26 = Add(A=BumpOffset(Coordinate=Multiply(A=@n9, B=@n10), Height=@n21, HeightRatioInput=@n22), B=@n25)
  n27 = Multiply(A=Time(), B=ScalarParameter() {defaultValue=0.025, parameterName=TexturePanningSpeed})
  n28 = Panner(Coordinate=Add(A=@n26, B=@n8), Time=@n27) {speedY=-0.3}
  n29 = Panner(Coordinate=@n26, Time=@n27) {speedY=-0.3}
  n30 = Add(A=@n26, B=Multiply(A=@n8) {constB=-1})
  n31 = AppendVector(
   A=AppendVector(
    A=TextureSample(UVs=@n28) {texture=/Game/Global/Textures/Default/Caustics/T_Turbulence_Seamless}.R
    B=TextureSample(UVs=@n29) {texture=&T_Turbulence_Seamless}.G
   )
   B=TextureSample(UVs=Panner(Coordinate=@n30, Time=@n27) {speedY=-0.3}) {texture=&T_Turbulence_Seamless}.B
  )
  n32 = AppendVector(A=ScalarParameter() {parameterName=Chroma_UOffset}, B=ScalarParameter() {defaultValue=0.025, parameterName=Chroma_VOffset})
  n33 = TextureCoordinate()
  n34 = AppendVector(A=ScalarParameter() {defaultValue=8, parameterName=TextureB_UScale}, B=ScalarParameter() {defaultValue=1.4, parameterName=TextureB_VScale})
  n35 = Add(A=@n25, B=BumpOffset(Coordinate=Multiply(A=@n33, B=@n34), Height=@n21, HeightRatioInput=@n22))
  n36 = Multiply(A=Time(), B=ScalarParameter() {defaultValue=0.025, parameterName=TexturePanningSpeed})
  n37 = Panner(Coordinate=Add(A=@n35, B=@n32), Time=@n36) {speedY=-0.3}
  n38 = Panner(Coordinate=@n35, Time=@n36) {speedY=-0.3}
  n39 = Add(A=@n35, B=Multiply(A=@n32) {constB=-1})
  n40 = AppendVector(
   A=AppendVector(
    A=TextureSample(UVs=@n37) {texture=&T_Turbulence_Seamless}.R
    B=TextureSample(UVs=@n38) {texture=&T_Turbulence_Seamless}.G
   )
   B=TextureSample(UVs=Panner(Coordinate=@n39, Time=@n36) {speedY=-0.3}) {texture=&T_Turbulence_Seamless}.B
  )
  n41 = LinearInterpolate(A=@n31, B=@n40)
  n42 = Add(
   A=ShadingPathSwitch(Default=@n31, Deferred=QualitySwitch(Default=@n31, High=@n41, Epic=@n41)) {desc=Migrated from FeatureLevelSwitch}
   B=Multiply(A=Power(Base=@n6, Exp=@n7), B=ScalarParameter() {defaultValue=0.35, parameterName=Fullness Intensity}) {constB=0.15}
  )
  n43 = ComponentMask(None=TextureCoordinate()) {g=true}
  n44 = ScalarParameter() {defaultValue=1.5, parameterName=Top Edge Alpha Contrast}
  n45 = Saturate(None=Multiply(A=Power(Base=@n43, Exp=@n44), B=ScalarParameter() {defaultValue=5, parameterName=Top Edge Alpha Mult}) {constB=5})
  n46 = MaterialFunctionCall!(
   Input=Length(None=WorldPosition() {worldPositionShaderOffset=WPT_CameraRelative}.XYZ) {is=class MaterialExpressionLength}
   Input Low=ScalarParameter() {defaultValue=300, parameterName=NearFadeDistanceMin}
   Input High=ScalarParameter() {defaultValue=2000, parameterName=NearFadeDistanceMax}
  ) {materialFunction=/Engine/Functions/Engine_MaterialFunctions03/Math/RemapValueRangeNormalized}
  n47 = MaterialFunctionCall(
   Brightness=ScalarParameter() {defaultValue=1, parameterName=UVFresnelBrightness}
   Exponent=ScalarParameter() {defaultValue=1, parameterName=UVFresnelExp}
  ) {materialFunction=&MF_Fresnel_FX}
  n48 = Multiply(A=Multiply(A=Multiply(A=@n42, B=@n45), B=@n46.Result), B=Saturate(None=@n47.Fresnel Double Sided))
  n49 = ScalarParameter() {defaultValue=1, parameterName=Opacity_Day_Mult}
  n50 = MaterialFunctionCall(Color_Morning=@n49, Color_Day=@n49, Color_Evening=@n49, Color_Night=ScalarParameter() {defaultValue=0.35, parameterName=Opacity_Night_Mult}) {materialFunction=&Color_Time_of_Day}
  n51 = Subtract(
   A=MaterialFunctionCall(Vector 3=ActorPositionWS() {originType=CameraRelative}) {materialFunction=/Engine/Functions/Engine_MaterialFunctions02/Utility/VectorLength}.V3 Length
   B=ScalarParameter() {parameterName=FarNear_BlendDistance}
  )
  n52 = ScalarParameter() {parameterName=FarNear_BlendRange}
  n53 = OneMinus(None=Saturate(None=Divide(A=@n51, B=@n52)))
  n54 = Length(None=WorldPosition() {worldPositionShaderOffset=WPT_CameraRelative}.XYZ)
  n55 = ScalarParameter() {defaultValue=1000, parameterName=OpacityInsideDistanceMin}
  n56 = ScalarParameter() {defaultValue=5000, parameterName=OpacityInsideDistanceMax}
  n57 = LinearInterpolate(
   A=LinearInterpolate(
    A=ScalarParameter() {defaultValue=1, parameterName=OpacityInsideMultClose}
    B=ScalarParameter() {defaultValue=0.5, parameterName=OpacityInsideMultFar}
    Alpha=MaterialFunctionCall(Input=@n54, Input Low=@n55, Input High=@n56) {materialFunction=&RemapValueRangeNormalized}.Result
   )
   B=ScalarParameter() {defaultValue=1, parameterName=OpacityOutside}
   Alpha=Saturate(None=TwoSidedSign())
  )
  n58 = ActorPositionWS() {originType=CameraRelative}
  n59 = WorldPosition() {worldPositionShaderOffset=WPT_CameraRelative}
  n60 = ScalarParameter() {defaultValue=35000, parameterName=FadeDist_WPO}
  n61 = ScalarParameter() {defaultValue=1.4, parameterName=HideDistanceBuffer}
  n62 = Constant()
  n63 = Multiply(
   A=Multiply(A=VectorParameter() {parameterName=TextureTint, defaultValue=<rgba 0.2 0.2267 1 0>}.RGB, B=@n42)
   B=ScalarParameter() {defaultValue=15, parameterName=BubbleBaseIntensity}
  )
  n64 = MakeMaterialAttributes!(
   BaseColor
   Metallic
   Specular
   Roughness
   Anisotropy
   EmissiveColor=StaticSwitchParameter(
    True=MaterialFunctionCall!(StartingHue, Base Emissive=@n63, Offset) {materialFunction=/Game/Environments/Props/RiftBeacon/FX/RiftBubble_GreenGoop}.TacoTimeResults
    False=@n63
   ) {parameterName=Allow TacoTime}
   Opacity=Multiply(A=Multiply(A=@n48, B=@n50.Result), B=Multiply(A=@n53, B=@n57))
   OpacityMask
   Normal
   Tangent
   WorldPositionOffset=If(
    A=MaterialFunctionCall(Vector 3=@n58, Vector 2=@n59.XY) {materialFunction=&VectorLength}.V3 Length
    B=Multiply(A=@n60, B=@n61) {constB=1.36}
    "A > B"=Constant3Vector() {constant=<rgba 0 0 -800000 1>}
    "A == B"=@n62
    "A < B"=@n62
   )
   SubsurfaceColor
   ClearCoat
   ClearCoatRoughness
   AmbientOcclusion
   Refraction
   CustomizedUVs_0
   CustomizedUVs_1
   CustomizedUVs_2
   CustomizedUVs_3
   CustomizedUVs_4
   CustomizedUVs_5
   CustomizedUVs_6
   CustomizedUVs_7
   PixelDepthOffset
   ShadingModel
   Displacement
  )
  n65 = Add(
   A=MaterialFunctionCall(Vector 3=WorldPosition() {worldPositionShaderOffset=WPT_CameraRelative}.XYZ) {materialFunction=&VectorLength}.V3 Length
   B=ScalarParameter() {parameterName=FarNear_BlendDistance}
  )
  n66 = ScalarParameter() {parameterName=FarNear_BlendRange}
  n67 = NamedRerouteUsage()
  n68 = VectorParameter() {parameterName=DayTint, defaultValue=<rgba 0.8 0.225 0.225 0>}
  n69 = MaterialFunctionCall(
   Brightness=ScalarParameter() {defaultValue=1, parameterName=Far_Fresnel Brightness}
   Exponent=ScalarParameter() {defaultValue=7, parameterName=Far_Fresnel Exponent}
  ) {materialFunction=&MF_Fresnel_FX}
  n70 = Multiply(
   A=MaterialFunctionCall(Color_Morning=@n68.RGB, Color_Day=@n68.RGB, Color_Evening=@n68.RGB, Color_Night=Constant() {r=1}) {materialFunction=&Color_Time_of_Day}.Result
   B=LinearInterpolate(
    A=VectorParameter() {parameterName=GlowColorOuter, defaultValue=<rgba 0.7 0.1768 0.0435 0>}.RGB
    B=VectorParameter() {parameterName=GlowColor, defaultValue=<rgba 1 0.4882 0 0>}.RGB
    Alpha=@n69.Fresnel
   )
  )
  n71 = StaticSwitchParameter(
   True=MaterialFunctionCall(Base Emissive=@n70) {materialFunction=&RiftBubble_GreenGoop}.TacoTimeResults
   False=@n70
  ) {parameterName=AllowTacoTime}
  n72 = Multiply(
   A=ScalarParameter() {defaultValue=0.2, parameterName=Far_OpacityMult}
   B=DepthFade(Opacity=@n69.Fresnel, FadeDistance=ScalarParameter() {defaultValue=50, parameterName=Far_DepthFade})
  )
  n73 = MaterialFunctionCall(Vector 3=ActorPositionWS() {originType=CameraRelative}) {materialFunction=&VectorLength}
  n74 = ScalarParameter() {parameterName=FarNear_BlendDistance}
  n75 = Saturate(None=Divide(A=Subtract(A=@n73.V3 Length, B=@n74), B=ScalarParameter() {parameterName=FarNear_BlendRange}))
  n76 = ActorPositionWS() {originType=CameraRelative}
  n77 = ScalarParameter() {defaultValue=35000, parameterName=FadeDist_WPO}
  n78 = ScalarParameter() {defaultValue=1.4, parameterName=HideDistanceBuffer}
  n79 = Constant()
  n80 = MakeMaterialAttributes(
   EmissiveColor=MaterialFunctionCall(Emissive=Add(A=@n67, B=@n71)) {materialFunction=&MF_FX_EmissiveAdjustments}.Result
   Opacity=Multiply(A=ScalarParameter() {defaultValue=1, parameterName=Far_Opacity}, B=Multiply(A=@n72, B=@n75))
   WorldPositionOffset=If(
    A=MaterialFunctionCall(Vector 3=@n76) {materialFunction=&VectorLength}.V3 Length
    B=Multiply(A=@n77, B=@n78) {constB=1.36}
    "A > B"=@n79
    "A == B"=@n79
    "A < B"=Constant3Vector() {constant=<rgba 0 0 -800000 1>}
   )
  )
  BlendMaterialAttributes(A=@n64, B=@n80, Alpha=Saturate(None=Divide(A=@n65, B=@n66)))
  Constant()
  LinearInterpolate(Alpha=@n47.Fresnel Double Sided) {constA=1.6, constB=-0.15}
  NamedRerouteDeclaration(None=Multiply(A=VectorParameter() {parameterName=TextureTint, defaultValue=<rgba 0.2 0.2267 1 0>}.RGB, B=@n40)) {name=EffectLow, nodeColor=<rgba 1 0.375 0 1>}
  OneMinus(None=@n46.Result)
  Saturate()
  StaticSwitchParameter(True=@n80, False=@n64) {parameterName=HLOD}
  WorldPosition() {worldPositionShaderOffset=WPT_CameraRelative}
  WorldPosition() {worldPositionShaderOffset=WPT_CameraRelative}
  WorldPosition() {worldPositionShaderOffset=WPT_CameraRelative}
 ]
}
